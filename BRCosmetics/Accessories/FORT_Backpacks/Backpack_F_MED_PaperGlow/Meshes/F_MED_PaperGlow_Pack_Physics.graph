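# Physics asset for the PaperGlow backpack cosmetic, giving its dangling sphere and ring-hung toy capsule bodies for secondary motion.
# The toy chain runs from dyn_Pack through two ring bodies to the toy and its end, locked in rotation at both ends with narrow swing and twist limits between.
# The sphere hangs separately off pack with its own limits.
# Large pelvis and spine_03 capsules link to pack with free rotation, apparently as body collision proxies.
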PhysicsAsset /BRCosmetics/Accessories/FORT_Backpacks/Backpack_F_MED_PaperGlow/Meshes/F_MED_PaperGlow_Pack_Physics
bodies {
 dyn_Pack [{shapeType=Capsule, length=21.1915, radius=1.1, center=<0 0 -0.5068>}]
 dyn_Ring_1 [{shapeType=Capsule, length=1, radius=1}]
 dyn_Ring_BACE_0 [{shapeType=Capsule, length=1, radius=1}]
 dyn_Sphere [{shapeType=Capsule, length=8.3004, radius=3.9705, shapeName=dyn_Sphere_capsule, center=<7.1378 0 0>, rotation=<rot 90 0 0>}]
 dyn_Toy [{shapeType=Capsule, length=2.7663, radius=3.814, center=<7.5098 0 0>, rotation=<rot -86.0882 0.0001 0>}]
 dyn_Toy_end [{shapeType=Capsule, length=1, radius=1}]
 pack [{shapeType=Capsule, length=1.9482, radius=1.7137, shapeName=pack_capsule}]
 pelvis [{shapeType=Capsule, length=15, radius=10.3309, shapeName=pelvis_capsule, center=<0.2881 -3.2874 0>}]
 spine_03 [{shapeType=Capsule, length=15, radius=15, shapeName=pelvis_capsule, center=<0 6.793 0>, rotation=<rot 90 0 -2.0733>}]
}
constraints [
 {
  bone1Name=dyn_Ring_BACE_0
  bone2Name=dyn_Pack
  swing1LimitDegrees=45
  swing1Motion=Locked
  swing2LimitDegrees=45
  swing2Motion=Locked
  twistLimitDegrees=45
  twistMotion=Locked
 }
 {bone1Name=dyn_Ring_1, bone2Name=dyn_Ring_BACE_0, swing1LimitDegrees=25, swing2LimitDegrees=15, twistLimitDegrees=3}
 {bone1Name=dyn_Toy, bone2Name=dyn_Ring_1, swing1LimitDegrees=20, swing2LimitDegrees=20, twistLimitDegrees=15}
 {
  bone1Name=dyn_Toy_end
  bone2Name=dyn_Toy
  swing1LimitDegrees=45
  swing1Motion=Locked
  swing2LimitDegrees=45
  swing2Motion=Locked
  twistLimitDegrees=45
  twistMotion=Locked
 }
 {
  bone1Name=pelvis
  bone2Name=pack
  swing1LimitDegrees=45
  swing1Motion=Free
  swing2LimitDegrees=45
  swing2Motion=Free
  twistLimitDegrees=45
  twistMotion=Free
 }
 {bone1Name=dyn_Sphere, bone2Name=pack, swing1LimitDegrees=35, swing2LimitDegrees=35, twistLimitDegrees=5}
 {
  bone1Name=spine_03
  bone2Name=pack
  swing1LimitDegrees=45
  swing1Motion=Free
  swing2LimitDegrees=45
  swing2Motion=Free
  twistLimitDegrees=45
  twistMotion=Free
 }
]
