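# Animation Blueprint for the PaperGlow backpack cosmetic, aligning it to the wearer and adding secondary motion.
# Each update copies the attach parent's pelvis and spine_03 world transforms and derives an additive pack offset from the owner's spine bend.
# The adjusted, cached pose feeds an AnimDynamics simulation of dyn_Sphere in the front end, or RigidBody physics with zero world gravity otherwise.
# A deprecated gravity-override state function also runs, but its stored result is unused here.
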AnimBlueprint /BRCosmetics/Accessories/FORT_Backpacks/Backpack_F_MED_PaperGlow/Meshes/F_MED_PaperGlow_Pack_AnimBP
target_skeleton=&/F_MED_PaperGlow_Pack_Skeleton
graphs [
 AnimGraph {nodes [
   n1 = Anim Dynamics - Bone: dyn_Sphere(ComponentPose=Local To Component(LocalPose=Use cached pose 'preSim'()), Alpha=0.8, ExternalForce="0,0,0") {
    angularDampingOverride=1
    angularSpringConstant=2000
    bAngularSpring=true
    bGravityOverrideInSimSpace=true
    bOverrideAngularDamping=true
    bUseGravityOverride=true
    simulationSpace=BoneRelative
    alphaBoolBlend {blendOption=Linear}
    boundBone {boneName=dyn_Sphere}
    componentLinearAccScale=<1 1 1>
    componentLinearVelScale=<1 1 1>
    gravityOverride {z=-980}
    physicsBodyDefinitions [{
      boundBone {boneName=dyn_Sphere}
      boxExtents=<10 10 10>
      constraintSetup {twistAxis=AxisY, angularLimitsMax=<10 10 15>, angularLimitsMin=<-10 -10 -15>, angularTarget=<0 0 0>}
      localJointOffset=<8 0 0>
     }]
    relativeSpaceBone {boneName=root}
   }
   n2 = RigidBody!(ComponentPose=Local To Component(LocalPose=Use cached pose 'preSim'()), Alpha, OverrideWorldGravity:Vector="0,0,0", ExternalForce) -> (Pose) {
    bClampLinearTranslationLimitToRefPose=true
    bOverrideWorldGravity=true
    overridePhysicsAsset=&/F_MED_PaperGlow_Pack_Physics
    simulationSpace=BaseBoneSpace
    alphaBoolBlend {blendOption=Linear}
    baseBoneRef {boneName=pack}
    componentAppliedLinearAccClamp=<20000 20000 20000>
    componentLinearAccScale=<0.3 0.3 0.3>
    componentLinearVelScale=<0.6 0.6 0.6>
    simSpaceSettings {
     maxAngularAcceleration=100
     maxAngularVelocity=2
     maxLinearAcceleration=150
     maxLinearVelocity=150
     velocityScaleZ=0
     worldAlpha=1
     externalLinearDragV=<0.1 0.1 0.1>
    }
   }
   Output Pose(Result=Blend Poses by bool(
     BlendPose_0=Component To Local(ComponentPose=@n1)
     BlendPose_1=Component To Local(ComponentPose=@n2)
     BlendTime_0=0
     bActiveValue=Get bIsInFrontEnd()
    ) {blendType=Linear}) {layerGroup=None}
   n3 = "Transform (Modify) Bone - Bone: pack"(ComponentPose=Local To Component(LocalPose=Input Pose()), Translation=Get backpack_translation(), Rotation=Get backpack_rotate()) {
    rotationMode=BMM_Additive
    rotationSpace=BCS_BoneSpace
    translationMode=BMM_Additive
    translationSpace=BCS_BoneSpace
    alphaBoolBlend {blendOption=Linear}
    boneToModify {boneName=pack}
   }
   n4 = "Transform (Modify) Bone - Bone: pelvis"(
    ComponentPose=@n3
    Alpha=1
    Translation=Get PelvisLoc() -> (PelvisLoc:Vector)
    Rotation=Get PelvisRot() -> (PelvisRot:Rotator)
    Scale="1,1,1"
   ) {
    rotationMode=BMM_Replace
    rotationSpace=BCS_WorldSpace
    scaleMode=BMM_Replace
    scaleSpace=BCS_WorldSpace
    translationMode=BMM_Replace
    translationSpace=BCS_WorldSpace
    alphaBoolBlend {blendInTime=0.2, blendOption=Linear, blendOutTime=0.2}
    boneToModify {boneName=pelvis}
    rotation {roll=25}
   }
   n5 = "Transform (Modify) Bone - Bone: spine_03"(ComponentPose=@n4, Alpha=1, Translation=Get SpineLoc() -> (SpineLoc:Vector), Rotation=Get SpineRot() -> (SpineRot:Rotator), Scale="1,1,1") {
    rotationMode=BMM_Replace
    rotationSpace=BCS_WorldSpace
    scaleMode=BMM_Replace
    scaleSpace=BCS_WorldSpace
    translationMode=BMM_Replace
    translationSpace=BCS_WorldSpace
    alphaBoolBlend {blendInTime=0.2, blendOption=Linear, blendOutTime=0.2}
    boneToModify {boneName=spine_03}
    rotation {roll=25}
   }
   Save cached pose 'preSim'(Pose=Component To Local(ComponentPose=@n5)) {cacheName=preSim}
  ]}
 EventGraph {nodes [
   Event BlueprintUpdateAnimation()
   > n1 = Sequence!(execute) -> (then_0, then_1, then_2:Exec)
   > n2 = Gravity Override State Function DEPRECATED(
    CharacterPartAnimInstance=Self-Reference()
    GameTimeInSeconds=GetGameTimeInSeconds()
    GravityOverrideParams=Get Gravity Override Params()
    DebugJointName=Pelvis
   )
   > Set Gravity Override(Gravity Override=@n2)
   > Set sky(sky:bool=Get bIsSkyDiving()) -> (Output_Get:bool)
   n3 = GetPartSkeletalMeshComponent()
   Is Valid(exec=@n1.then_1, InputObject=@n3)
   n4 = GetAttachParent!(self:Scene Component Object Reference=@n3) -> (ReturnValue:Scene Component Object Reference)
   > Is Valid(InputObject=@n4)
   n5 = GetSocketTransform!(self=@n4, InSocketName=pelvis, TransformSpace=RTS_World) -> (ReturnValue_Location:Vector, ReturnValue_Rotation:Rotator, ReturnValue_Scale:Vector)
   > Set PelvisLoc(PelvisLoc:Vector=@n5) -> (Output_Get:Vector)
   > Set PelvisRot(PelvisRot:Rotator=@n5.ReturnValue_Rotation) -> (Output_Get:Rotator)
   n6 = GetSocketTransform(self=@n4, InSocketName=spine_03, TransformSpace=RTS_World)
   > Set SpineLoc(SpineLoc:Vector=@n6) -> (Output_Get:Vector)
   > Set SpineRot(SpineRot:Rotator=@n6.ReturnValue_Rotation) -> (Output_Get:Rotator)
   n7 = GetOwnerSkeletalMeshComponent()
   n8 = BreakRotator(InRot=BreakTransform(InTransform=GetDeltaTransformFromRefPose(self=@n7, BoneName=spine_05, BaseName=spine_01)).Rotation)
   Set backpack_rotate(execute=@n1.then_2, backpack_rotate=MakeRotator(Roll=MapRangeClamped(Value=@n8.Yaw, InRangeA=0, InRangeB=-40, OutRangeB=20)))
   > Set backpack_translation(backpack_translation=MakeVector(
     Y=MapRangeClamped(Value=@n8.Yaw, InRangeA=0, InRangeB=-130, OutRangeB=-4)
     Z=MapRangeClamped(Value=@n8.Yaw, InRangeA=0, InRangeB=-130, OutRangeB=2)
    ))
   n9 = Get bIsInFrontEnd()
   > Set isFrontEnd(isFrontEnd:double="To Float (Boolean)"(InBool=@n9)) -> (Output_Get:double)
   > Set isNotFrontEnd(isNotFrontEnd:double="To Float (Boolean)"(InBool=NOT Boolean(A=@n9))) -> (Output_Get:double)
   TryGetPawnOwner()
  ]}
]
variables [
 backpack_rotate {type=Rotator}
 backpack_translation {type=Vector}
 Gravity Override {type=Vector}
 Gravity Override Params {type=GravityOverrideParamsStruct}
 isFrontEnd {type=number}
 isNotFrontEnd {type=number}
 PelvisLoc {type=Vector}
 PelvisRot {type=Rotator}
 sky {type=boolean}
 SpineLoc {type=Vector}
 SpineRot {type=Rotator}
]
